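# Stereo MetaSound source for the Puppeteer boss mask's idle ambience, with attenuation and single-instance stop-oldest concurrency.
# On play, it starts a void loop bed and two independent one-shot layers, each picking random waves after random delays.
# Each layer re-arms itself by feeding its player's On Finished through a deferred trigger variable into Trigger Any.
# A three-input stereo mixer sums the bed and both layers to output.
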MetaSoundSource /PuppeteerBoss/Audio/SFX/Mask/MSS_Puppeteer_Mask_Idle_Loop
inputs [
 OneShots_A {type=WaveAsset:Array}
 OneShots_B {type=WaveAsset:Array}
 UE.Source.OnPlay {type=Trigger}
 Wave Asset {type=WaveAsset:Array}
]
interfaces [UE.OutputFormat.Stereo:1.0, UE.Source:1.1]
outputs [
 "UE.OutputFormat.Stereo.Audio:0" {type=Audio}
 "UE.OutputFormat.Stereo.Audio:1" {type=Audio}
]
pages [Default {
  input_defaults {
   UE.Source.OnPlay=null
   OneShots_A [
    /PuppeteerBoss/Audio/SFX/Mask/wav/Puppeteer_Mask_Idle_OS_01
    /PuppeteerBoss/Audio/SFX/Mask/wav/Puppeteer_Mask_Idle_OS_02
    /PuppeteerBoss/Audio/SFX/Mask/wav/Puppeteer_Mask_Idle_OS_03
   ]
   OneShots_B [
    /PuppeteerBoss/Audio/SFX/Mask/wav/Puppeteer_Mask_Idle_OS_04
    /PuppeteerBoss/Audio/SFX/Mask/wav/Puppeteer_Mask_Idle_OS_05
    /PuppeteerBoss/Audio/SFX/Mask/wav/Puppeteer_Mask_Idle_OS_06
   ]
   Wave Asset [/PuppeteerBoss/Audio/SFX/Mask/wav/Puppeteer_Mask_Void_Loop_Boss_01]
  }
  nodes [
   n1 = VariableDeferredAccessor.Trigger(Variable=InitVariable.Trigger().Variable)
   TriggerAnyTriggerAny2_0 = "TriggerAny.Trigger Any (2)"(In 0=^UE.Source.OnPlay, In 1=@n1.Value)
   n2 = UE.RandomTime(Next=@TriggerAnyTriggerAny2_0.Out) {also {defaults {Max=10, Min=5}}}
   TriggerDelay_1 = UE.Trigger Delay(In=@n2.On Next, Delay Time=@n2.Value)
   ArrayRandomGet_0 = Array.Random Get:WaveAsset:Array(In Array=^OneShots_B, Next=@TriggerDelay_1.Out) {also {defaults {Weights []}}}
   WavePlayer_0 = UE.Wave Player:Stereo(Play=@ArrayRandomGet_0.On Next, Wave Asset=@ArrayRandomGet_0.Value)
   n3 = VariableDeferredAccessor.Trigger(Variable=InitVariable.Trigger().Variable)
   TriggerAnyTriggerAny2_1 = "TriggerAny.Trigger Any (2)"(In 0=^UE.Source.OnPlay, In 1=@n3.Value)
   n4 = UE.RandomTime(Next=@TriggerAnyTriggerAny2_1.Out) {also {defaults {Max=8, Min=4}}}
   TriggerDelay_0 = UE.Trigger Delay(In=@n4.On Next, Delay Time=@n4.Value)
   ArrayRandomGet_1 = Array.Random Get:WaveAsset:Array(In Array=^OneShots_A, Next=@TriggerDelay_0.Out) {also {defaults {Weights []}}}
   WavePlayer_1 = UE.Wave Player:Stereo(Play=@ArrayRandomGet_1.On Next, Wave Asset=@ArrayRandomGet_1.Value)
   ArrayRandomGet_2 = Array.Random Get:WaveAsset:Array(Next=^UE.Source.OnPlay, In Array=^Wave Asset)
   WavePlayer_2 = UE.Wave Player:Stereo(Play=@ArrayRandomGet_2.On Next, Wave Asset=@ArrayRandomGet_2.Value) {also {defaults {Loop="<MetaSoundBuilderResult.SUCCEEDED: 0>"}}}
   AudioMixerAudioMixerStereo3_0 = "AudioMixer.Audio Mixer (Stereo, 3)"(
    In 2 L=@WavePlayer_0.Out Left
    In 2 R=@WavePlayer_0.Out Right
    In 1 L=@WavePlayer_1.Out Left
    In 1 R=@WavePlayer_1.Out Right
    In 0 L=@WavePlayer_2.Out Left
    In 0 R=@WavePlayer_2.Out Right
   ) {also {defaults {Gain 0=1.1}}}
   VariableMutator.Trigger(Variable=@n1.Variable, Value=@WavePlayer_0.On Finished)
   VariableMutator.Trigger(Variable=@n3.Variable, Value=@WavePlayer_1.On Finished)
  ]
  links [
   {dst="@UE.OutputFormat.Stereo.Audio:0", src=AudioMixerAudioMixerStereo3_0.Out L}
   {dst="@UE.OutputFormat.Stereo.Audio:1", src=AudioMixerAudioMixerStereo3_0.Out R}
  ]
 }]
properties {
 assetImportData=&self:AssetImportData
 attenuationSettings=&/Att_PuppeteerMask_Idle
 bLooping=true
 bOverrideConcurrency=true
 duration=10000
 maxDistance=4000
 numChannels=2
 outputFormat=Stereo
 sampleRate=48000
 soundClassObject=&Fort_BossFight
 concurrencyOverrides {maxCount=1, resolutionRule=StopOldest, retriggerTime=1}
}
